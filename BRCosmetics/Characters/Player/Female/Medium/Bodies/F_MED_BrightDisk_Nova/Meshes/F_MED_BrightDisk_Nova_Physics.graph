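# Physics asset for the F_MED_BrightDisk_Nova female body, set up for secondary motion on its jacket and semi-collar.
# Small base capsules anchor jacket and right collar bones through narrow angular limits, with the upper-left jacket forming a two-link chain.
# Head, upper arms and right clavicle add capsule bodies around the cloth.
# The upper-right jacket also has an angularly locked constraint to the right clavicle, whose upper-arm joint is free.
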PhysicsAsset /BRCosmetics/Characters/Player/Female/Medium/Bodies/F_MED_BrightDisk_Nova/Meshes/F_MED_BrightDisk_Nova_Physics
bodies {
 Base_low_jacket_l [{shapeType=Capsule, length=0.505, radius=0.505, shapeName=Base_low_jacket_l_capsule, rotation=<rot 0 -90 0>}]
 base_semi_collar_r [{shapeType=Capsule, length=0.505, radius=0.505, shapeName=base_semi_collar_r_capsule, rotation=<rot 0 -90 0>}]
 base_up_jacket_l [{shapeType=Capsule, length=0.505, radius=0.505, shapeName=base_up_jacket_l_capsule, rotation=<rot 0 -90 0>}]
 base_up_jacket_r [{shapeType=Capsule, length=0.505, radius=0.505, shapeName=base_up_jacket_r_capsule, rotation=<rot 0 -90 0>}]
 clavicle_r [{shapeType=Capsule, length=5.6887, radius=5.1235, center=<-10.0871 0.2372 -0.4768>, rotation=<rot 90 0 0>}]
 cor_semi_collar_r_1 [{
   shapeType=Capsule
   length=1.9635
   radius=1.6832
   shapeName=cor_semi_collar_r_1_capsule
   center=<-1.6029 0.6411 -1.0437>
   rotation=<rot 0 -126.9244 -86.8436>
  }]
 dyn_low_jacket_l [{
   shapeType=Capsule
   length=6.8415
   radius=1.9336
   shapeName=dyn_low_jacket_l_capsule
   center=<6.012 -0.6261 -4.4161>
   rotation=<rot 0 -90.1301 -118.3413>
  }]
 dyn_up_jacket_01_l [{
   shapeType=Box
   extentX=9.5789
   extentY=12.7993
   extentZ=2.3738
   shapeName=dyn_up_jacket_01_l_box
   center=<2.3951 -2.7902 -2.3835>
   rotation=<rot -32.4932 -3.0153 -30.3607>
  }]
 dyn_up_jacket_01_r [{
   shapeType=Box
   extentX=16.3492
   extentY=2.0251
   extentZ=17.3181
   shapeName=dyn_up_jacket_01_r_box
   center=<6.7459 -1.2867 -2.2042>
   rotation=<rot -15.8711 -140.9191 -71.4908>
  }]
 dyn_up_jacket_02_l [{shapeType=Capsule, length=11.9066, radius=3.214, center=<0.9855 2.7368 2.7096>, rotation=<rot 32.4361 53.7126 34.5821>}]
 head [{shapeType=Capsule, length=0, radius=11.3669, center=<3.7824 1.2609 -0.2775>, rotation=<rot 1.5491 -90.641 2.3565>}]
 upperarm_l [{shapeType=Capsule, length=18.4229, radius=5.9567, center=<9.9905 -0.3855 0.4613>, rotation=<rot -83.5984 179.5742 176.7871>}]
 upperarm_r [{shapeType=Capsule, length=15.857, radius=5.1497, center=<-9.9432 0.4439 -0.1676>, rotation=<rot -81.8524 179.2989 176.1405>}]
}
constraints [
 {bone1Name=dyn_up_jacket_01_r, bone2Name=base_up_jacket_r, swing1LimitDegrees=10, swing2LimitDegrees=10, twistLimitDegrees=0}
 {
  bone1Name=upperarm_r
  bone2Name=clavicle_r
  swing1LimitDegrees=45
  swing1Motion=Free
  swing2LimitDegrees=45
  swing2Motion=Free
  twistLimitDegrees=45
  twistMotion=Free
 }
 {bone1Name=dyn_up_jacket_01_l, bone2Name=base_up_jacket_l, swing1LimitDegrees=0, swing2LimitDegrees=10, twistLimitDegrees=0}
 {bone1Name=dyn_up_jacket_02_l, bone2Name=dyn_up_jacket_01_l, swing1LimitDegrees=0, swing2LimitDegrees=10, twistLimitDegrees=0}
 {bone1Name=dyn_low_jacket_l, bone2Name=Base_low_jacket_l, swing1LimitDegrees=5, swing2LimitDegrees=5, twistLimitDegrees=0}
 {
  bone1Name=dyn_up_jacket_01_r
  bone2Name=clavicle_r
  swing1LimitDegrees=45
  swing1Motion=Locked
  swing2LimitDegrees=45
  swing2Motion=Locked
  twistLimitDegrees=45
  twistMotion=Locked
 }
 {bone1Name=cor_semi_collar_r_1, bone2Name=base_semi_collar_r, swing1LimitDegrees=1, swing2LimitDegrees=1, twistLimitDegrees=1}
]
